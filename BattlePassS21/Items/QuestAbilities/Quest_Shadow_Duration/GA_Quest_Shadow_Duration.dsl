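; Battle Pass S21 quest gameplay ability that reports time spent in shadow (Shadow form or state) to the quest system.
; Event activation calls the parent, then starts a non-looping timer of ExecutionTimeInterval.
; When the timer fires, it sends a count of 1 under the T_Quest tag to the owning player's quest manager.
; Ability end resets ElapsedTimeInShadow and clears the timer; debug strings print on start and end.
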(event Ability|EventActivateAbilityFromEvent
  (params (EventData :type Gameplay%20Event%20Data%20Structure%20%28by%20ref%29))
  (|CallParentFunction EventData)
  (Development|PrintString "Shadow Started")
  (bind _returnvalue (Utilities|Time|SetTimerbyEvent (delegate Custom|Check%20Shadow%20Time) (Variables|Default|GetExecutionTimeInterval)))
  (Variables|Default|SetTimer%20Handle _returnvalue))

(event Custom|Check%20Shadow%20Time
  (Quest|SendComplexCustomStatEvent (Variables|GAVars|Data|GetOwningPlayerQuestManager) 0 (Variables|Default|GetT_Quest) "(GameplayTags=)" 1))

(event Ability|EventOnEndAbility
  (params (bWasCancelled :type Boolean))
  (bind _timer_handle (Variables|Default|GetTimer%20Handle))
  ;[1 1 1] On End Ability
  (Variables|Default|SetElapsedTimeInShadow 0.0)
  (if (Utilities|Time|IsValidTimerHandle _timer_handle)
    (Utilities|Time|ClearandInvalidateTimerbyHandle _timer_handle)
    (else
      _))
  (Ability|EndAbility)
  (Development|PrintString "shadow ended"))
